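# Niagara effect for the QuickBurst speakers gadget cosmetic: a brief fire puff with a double electric shockwave pulse.
# Fire bursts short-lived sprites pushed by curl noise, driving appearance through curve-animated dynamic material parameters.
# Pulse_ElectricArcs fires two timed bursts split into spawn groups that set mesh scale, velocity and visibility tags across three mesh renderers, with separate glow and shockwave parameters.
# Fire_LWE is disabled, and no user parameters are exposed.
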NiagaraSystem /BRCosmetics/Animation/Props/Gadgets_QuickBurst_Speakers_Left/FX/NS_QuickBurst_Speakers
dependencies {
 usedDataInterfaces [&NiagaraDataInterfaceCurve]
 usedDynamicInputs []
 usedModules [
  &self:SystemSpawnScript.SystemScriptSource.SystemScriptGraph.NiagaraNodeAssignment_0.SetVariables_#1
  &SystemState
 ]
}
emitters {
 Fire {
  emitterUpdateScript [
   &EmitterState()
   &SpawnBurst_Instantaneous(Spawn Count=10, Use Spawn Probability=false, UseLoopCountLimit=false)
  ]
  particleSpawnScript [
   &InitializeParticle(
    Lifetime Max=0.3
    Lifetime Min=0.2
    Lifetime Mode=::NewEnumerator1
    Sprite Rotation Mode=::NewEnumerator2
    Sprite Size Mode=::NewEnumerator1
    Sprite UV Mode=::NewEnumerator3
    Uniform Sprite Size Max=15
    Uniform Sprite Size Min=12
    UsePositionOffset=false
    Write Material Random=true
   )
   &ShapeLocation(
    Handle Radius=0
    Large Radius=5
    MaskSpawnGroup=false
    Offset Mode=::NewEnumerator0
    Override Seed=false
    Rotation Mode=::NewEnumerator1
    Shape Primitive=::NewEnumerator3
    UseOwnerScale=true
   )
   &AddVelocity(Module.Use Falloff Distance=false, Origin Offset=<-10 0 0>, Override Seed=false, Velocity Mode=::NewEnumerator1, Velocity Speed=50)
  ]
  particleUpdateScript [
   &CurlNoiseForce(
    Noise Frequency=250
    Noise Quality / Cost=::NewEnumerator0
    Noise Strength=&FloatFromCurve(
     FloatCurve=DataInterface() {propertyValues {Curve {keys [{tangentMode=RCTM_User, time=0.3, value=0}, {tangentMode=RCTM_User, time=0.4, value=1}]}}}
     CurveIndex=$Emitter.Age
     Scale Curve=&RandomRangeFloat(Minimum=200, Maximum=250)
    )
    Pan Noise=true
    Pan Noise Field=<0.6 0.6 1>
   )
   &SolveForcesAndVelocity(Clamp Velocity=false, Limit Acceleration=false)
   &ParticleState()
   &ScaleSpriteSize(Uniform Curve Sprite Scale=DataInterface() {propertyValues {Curve {keys [
        {arriveTangent=0.7331, interpMode=RCIM_Cubic, leaveTangent=0.7331, tangentMode=RCTM_User, time=0, value=0.25}
        {arriveTangent=2.5631, interpMode=RCIM_Cubic, leaveTangent=2.5631, tangentMode=RCTM_User, time=0.2, value=0.5}
        <at 0.7 1.25 RCIM_Cubic>
       ]}}})
   &DynamicMaterialParameters(
    Index 0 Param 1=&Add_Float(
     A=&RandomRangeFloat(Minimum=0, Maximum=5)
     B=&FloatFromCurve(
      FloatCurve=DataInterface() {propertyValues {Curve {keys [
          {arriveTangent=-2.2561, interpMode=RCIM_Cubic, leaveTangent=-2.2561, tangentMode=RCTM_User, time=0, value=1}
          {arriveTangent=-1.0036, interpMode=RCIM_Cubic, leaveTangent=-1.0036, tangentMode=RCTM_User, time=0.2682, value=0.3}
          <at 0.8 0>
         ]}}}
      Scale Curve=0.5
     )
    )
    Index 0 Param 2=&FloatFromCurve(
     FloatCurve=DataInterface() {propertyValues {Curve {keys [
         {arriveTangent=0.086, interpMode=RCIM_Cubic, leaveTangent=0.086, tangentMode=RCTM_User, time=0.15, value=0.2}
         {arriveTangent=2.0528, interpMode=RCIM_Cubic, leaveTangent=2.0528, tangentMode=RCTM_User, time=1, value=1}
        ]}}}
     CurveIndex=$Emitter.Age
     Scale Curve=0.4
    )
    Index 0 Param 3=1
    Index 1 Param 1=&FloatFromCurve(
     FloatCurve=DataInterface() {propertyValues {Curve {keys [
         {arriveTangent=-1.3223, interpMode=RCIM_Cubic, leaveTangent=-1.3223, tangentMode=RCTM_User, time=0.1, value=0.5}
         {arriveTangentWeight=0.3, interpMode=RCIM_Cubic, time=1, value=0}
        ]}}}
     Scale Curve=&FloatFromCurve(FloatCurve=DataInterface() {propertyValues {Curve {keys [<at 0.1 15>,<at 0.3 1>,<at 1 1>]}}})
    )
    Index 1 Param 2=&FloatFromCurve(
     FloatCurve=DataInterface() {propertyValues {Curve {keys [{arriveTangent=-1.2376, interpMode=RCIM_Cubic, leaveTangent=-1.2376, tangentMode=RCTM_User, time=0, value=1}, <at 1 0>]}}}
     Scale Curve=10
    )
    Index 1 Param 3=&FloatFromCurve(
     FloatCurve=DataInterface() {propertyValues {Curve {keys [
         {interpMode=RCIM_Cubic, tangentMode=RCTM_User, time=0.333, value=1.5}
         {arriveTangent=0.7571, interpMode=RCIM_Cubic, leaveTangent=0.7571, tangentMode=RCTM_User, time=1, value=1.8}
        ]}}}
     CurveIndex=$Emitter.Age
     Scale Curve=1.35
    )
    Index 1 Param 4=&FloatFromCurve(
     FloatCurve=DataInterface() {propertyValues {Curve {keys [<at 0 0.5>,<at 1 0.3>]}}}
     Scale Curve=&Add_Float(A=1, B=&RandomRangeFloat(Minimum=0, Maximum=5))
    )
    Index 2 Param 1=&FloatFromCurve(FloatCurve=DataInterface() {propertyValues {Curve {keys [{arriveTangent=0.0048, interpMode=RCIM_Cubic, leaveTangent=0.0048, tangentMode=RCTM_User, time=0, value=0.8}, <at 0.9 1.2 RCIM_Cubic>]}}})
    Index 2 Param 2=&Add_Float(
     A=&RandomRangeFloat(Minimum=0, Maximum=100)
     B=&FloatFromCurve(
      FloatCurve=DataInterface() {propertyValues {Curve {keys [
          {arriveTangent=0.5501, interpMode=RCIM_Cubic, leaveTangent=0.5501, tangentMode=RCTM_User, time=0, value=0}
          {arriveTangent=1.1924, interpMode=RCIM_Cubic, leaveTangent=1.1924, tangentMode=RCTM_User, time=1, value=0.8}
         ]}}}
      Scale Curve=&RandomRangeFloat(Minimum=-0.3, Maximum=-0.56)
     )
    )
    Index 2 Param 3=&FloatFromCurve(FloatCurve=DataInterface() {propertyValues {Curve {keys [
         {arriveTangent=-0.4045, interpMode=RCIM_Cubic, tangentMode=RCTM_User, time=0.4, value=0.5}
         {arriveTangent=5.0157, interpMode=RCIM_Cubic, leaveTangent=5.0157, tangentMode=RCTM_User, time=1, value=2}
        ]}}})
    Index 2 Param 4=&FloatFromCurve(FloatCurve=DataInterface() {propertyValues {Curve {keys [{arriveTangent=-0.4045, time=0, value=0.8}, {leaveTangent=-0.318, time=1, value=1}]}}})
    Index 3 Param 1=&FloatFromCurve(
     FloatCurve=DataInterface() {propertyValues {Curve {keys [
         {arriveTangent=0.1645, interpMode=RCIM_Cubic, leaveTangent=0.1645, tangentMode=RCTM_User, time=0.1, value=0.2}
         {arriveTangent=2.6313, interpMode=RCIM_Cubic, leaveTangent=2.6313, tangentMode=RCTM_User, time=0.6, value=1.25}
         <at 1 2 RCIM_Cubic>
        ]}}}
     Scale Curve=4
    )
    Index 3 Param 2=&FloatFromCurve(
     FloatCurve=DataInterface() {propertyValues {Curve {keys [<at 0 1>,<at 1 0>]}}}
     Scale Curve=&RandomRangeFloat(Minimum=0.4, Maximum=0.5)
    )
    Index 3 Param 3=&FloatFromCurve(
     FloatCurve=DataInterface() {propertyValues {Curve {keys [
         {arriveTangent=0.4884, interpMode=RCIM_Cubic, leaveTangent=0.4884, tangentMode=RCTM_User, time=0, value=0.7}
         {arriveTangent=0.3564, interpMode=RCIM_Cubic, leaveTangent=0.3564, tangentMode=RCTM_User, time=1, value=0.9}
        ]}}}
     Scale Curve=0.0009
    )
    Index 3 Param 4=0.25
    Module.Index0Param0Write=true
    Module.Index0Param1Write=true
    Module.Index0Param2Write=true
    Module.Index0Param3Write=false
    Module.Index1Param0Write=true
    Module.Index1Param1Write=true
    Module.Index1Param2Write=true
    Module.Index1Param3Write=true
    Module.Index2Param0Write=true
    Module.Index2Param1Write=true
    Module.Index2Param2Write=true
    Module.Index2Param3Write=true
    Module.Index3Param0Write=true
    Module.Index3Param1Write=true
    Module.Index3Param2Write=true
    Module.Index3Param3Write=true
    Write Parameter Index 1=true
    Write Parameter Index 2=true
    Write Parameter Index 3=true
   )
   &ScaleColor(
    Scale Alpha=&FloatFromCurve(
     FloatCurve=DataInterface() {propertyValues {Curve {keys [
         <at 0 0.05>
         {arriveTangentWeight=0.846, interpMode=RCIM_Cubic, leaveTangentWeight=0.6768, tangentMode=RCTM_User, time=0.6, value=0.95}
         {interpMode=RCIM_Cubic, tangentMode=RCTM_Break, time=0.9, value=0}
        ]}}}
     Scale Curve=&FloatFromCurve(FloatCurve=DataInterface() {propertyValues {Curve {keys [{arriveTangent=-4.6812, interpMode=RCIM_Cubic, leaveTangent=-4.6812, tangentMode=RCTM_User, time=0.5, value=1}, <at 1 0 RCIM_Cubic>]}}})
    )
    ScaleA=true
    ScaleRGB=false
   )
  ]
  rendererClasses [&NiagaraSpriteRendererProperties]
  renderers [{rendererClass=&NiagaraSpriteRendererProperties}]
 }
 Fire_LWE {bEnabled=false, rendererClasses [], renderers []}
 Pulse_ElectricArcs {
  emitterUpdateScript [
   &EmitterState()
   &SpawnBurst_Instantaneous(Use Spawn Probability=false, UseLoopCountLimit=false)
   &SpawnBurst_Instantaneous(Spawn Time=0.1, Use Spawn Probability=false, UseLoopCountLimit=false) {module=SpawnBurst_Instantaneous001}
   &SpawnBurst_Instantaneous(Spawn Count=2, Spawn Group=1, Use Spawn Probability=false, UseLoopCountLimit=false) {module=SpawnBurst_Instantaneous002}
   &SpawnBurst_Instantaneous(Spawn Count=2, Spawn Group=1, Spawn Time=0.1, Use Spawn Probability=false, UseLoopCountLimit=false) {module=SpawnBurst_Instantaneous003}
  ]
  particleSpawnScript [
   /SetVariables(Particles.VisibilityTag=&SpawnGroup()) {isSetParameters=true}
   &InitializeParticle(
    Color Maximum=<rgba 0.18 0.18 0.18 0.7>
    Color Minimum=<rgba 0.075 0.075 0.075 0.4>
    Color Mode=::NewEnumerator2
    Lifetime=0.3
    Mesh Scale Mode=::NewEnumerator3
    Mesh Uniform Scale=&MaskFloatBySpawnGroup(Masked Float=0.6, Passthrough Float=&RandomRangeFloat(Minimum=9, Maximum=12))
    UsePositionOffset=false
    Write Material Random=true
   )
   &InitialMeshOrientation(
    Mesh Axis To Orient=<0 0 1>
    Mesh Orientation Mode=::NewEnumerator2
    Rotation=&RandomRangeVector(Minimum=<0 0 0>, Maximum=<0 0 1>)
    Use Rotation Vector=true
   )
   &AddVelocity(Velocity=&MaskVectorBySpawnGroup(Masked Vector=<10 0 0>, Passthrough Vector=<50 0 0>))
  ]
  particleUpdateScript [
   &SolveForcesAndVelocity(Clamp Velocity=false, Limit Acceleration=false)
   &ParticleState()
   &UpdateMeshOrientation(Rotation Rate=&RandomRangeFloat(Minimum=-0.15, Maximum=0.15), Rotation Vector=<0 0 1>)
   &ScaleMeshSize(Scale Factor=&VectorFromFloat(Value=&MaskFloatBySpawnGroup(
      Masked Float=&FloatFromCurve(FloatCurve=DataInterface() {propertyValues {Curve {keys [{arriveTangent=1.1031, interpMode=RCIM_Cubic, leaveTangent=1.1031, tangentMode=RCTM_User, time=0, value=0.25}, <at 1 1.25 RCIM_Cubic>]}}})
      Passthrough Float=&FloatFromCurve(FloatCurve=DataInterface() {propertyValues {Curve {keys [
           {arriveTangent=5.1547, interpMode=RCIM_Cubic, leaveTangent=5.1547, leaveTangentWeight=0.7454, tangentMode=RCTM_User, time=0, value=1}
           {arriveTangentWeight=0.7454, interpMode=RCIM_Cubic, tangentMode=RCTM_User, time=1, value=3}
          ]}}})
     )))
   &DynamicMaterialParameters(
    Index 0 Param 4=13
    Index 1 Param 1=&FloatFromCurve(FloatCurve=DataInterface() {propertyValues {Curve {keys [<at 0 0.5>,<at 1 1>]}}})
    Index 1 Param 2=2
    Index 1 Param 3=&FloatFromCurve(FloatCurve=DataInterface() {propertyValues {Curve {keys [{interpMode=RCIM_Cubic, tangentMode=RCTM_Break, time=0.5, value=-0.5}, <at 1 0.1>]}}})
    Index 2 Param 1=&Multiply_Float(
     A=&FloatFromCurve(
      FloatCurve=DataInterface() {propertyValues {Curve {keys [{interpMode=RCIM_Cubic, leaveTangent=8.1085, tangentMode=RCTM_Break, time=0.8568, value=1}, {tangentMode=RCTM_User, time=1, value=0}]}}}
      Scale Curve=0.125
     )
     B=&Divide_Float(
      A=&MakeFloatFromVector(Vector=$Particles.Scale, Channel=::NewEnumerator0)
      B=&RandomRangeFloat(Minimum=10, Maximum=15)
     )
    )
    Index 2 Param 2=&FloatFromCurve(
     FloatCurve=DataInterface() {propertyValues {Curve {keys [
         {interpMode=RCIM_Cubic, tangentMode=RCTM_Break, time=0.7, value=0}
         {arriveTangent=-0.0859, interpMode=RCIM_Cubic, tangentMode=RCTM_Break, time=1, value=-1}
        ]}}}
     Scale Curve=0
    )
    Index 2 Param 3=&Waveform(
     Global Input Position=&RandomRangeFloat(Minimum=0, Maximum=15)
     Global Phase=$Particles.Age
     "Waveform [1]"=::NewEnumerator12
     "[1] Frequency"=1
     "[1] Amplitude Min/Max"=<0 1>
    )
    Index 2 Param 4=75
    Index 3 Param 1=&FloatFromCurve(
     FloatCurve=DataInterface() {propertyValues {Curve {keys [
         {interpMode=RCIM_Cubic, leaveTangent=-0.0025, tangentMode=RCTM_Break, time=0, value=0}
         {arriveTangent=1.8528, interpMode=RCIM_Cubic, tangentMode=RCTM_Break, time=1, value=1}
        ]}}}
     Scale Curve=&RandomRangeFloat(Minimum=0, Maximum=0)
    )
    Index 3 Param 2=&FloatFromCurve(
     FloatCurve=DataInterface() {propertyValues {Curve {keys [
         {interpMode=RCIM_Cubic, leaveTangent=-0.0025, tangentMode=RCTM_Break, time=0, value=0.5}
         {arriveTangent=-0.0316, interpMode=RCIM_Cubic, tangentMode=RCTM_Break, time=1, value=-0.2}
        ]}}}
     Scale Curve=0
    )
    Index 3 Param 3=&FloatFromCurve(FloatCurve=DataInterface() {propertyValues {Curve {keys [
         {interpMode=RCIM_Cubic, leaveTangent=1.1034, tangentMode=RCTM_Break, time=0, value=-1}
         {arriveTangent=2.2399, tangentMode=RCTM_User, time=0.5, value=0.125}
        ]}}})
    Index 3 Param 4=&FloatFromCurve(FloatCurve=DataInterface() {propertyValues {Curve {keys [
         {interpMode=RCIM_Cubic, leaveTangent=1.8089, tangentMode=RCTM_Break, time=0.3, value=-0.5}
         {arriveTangent=2.0757, interpMode=RCIM_Cubic, tangentMode=RCTM_Break, time=0.9, value=1}
        ]}}})
    Module.Index0Param0Write=true
    Module.Index0Param1Write=true
    Module.Index0Param2Write=true
    Module.Index0Param3Write=true
    Module.Index1Param0Write=true
    Module.Index1Param1Write=true
    Module.Index1Param2Write=true
    Module.Index1Param3Write=true
    Module.Index2Param0Write=true
    Module.Index2Param1Write=true
    Module.Index2Param2Write=true
    Module.Index2Param3Write=true
    Module.Index3Param0Write=true
    Module.Index3Param1Write=true
    Module.Index3Param2Write=true
    Module.Index3Param3Write=true
    Write Parameter Index 1=true
    Write Parameter Index 2=true
    Write Parameter Index 3=true
   )
   &ScaleColor(
    Scale Alpha=&FloatFromCurve(FloatCurve=DataInterface() {propertyValues {Curve {keys [<at 0 1 RCIM_Cubic>, {interpMode=RCIM_Cubic, leaveTangent=0.1205, time=1, value=0}]}}})
    Scale RGB=&VectorFromFloat(Value=&Lerp_Float(
      A=20
      B=&RandomRangeFloat(Minimum=-0.5, Maximum=2, Evaluation Type=::NewEnumerator1)
      Alpha=&FloatFromCurve(FloatCurve=DataInterface() {propertyValues {Curve {keys [
           {arriveTangent=1.0951, interpMode=RCIM_Cubic, leaveTangent=1.0951, leaveTangentWeight=0.3923, tangentMode=RCTM_User, time=0, value=0}
           {arriveTangentWeight=0.3923, interpMode=RCIM_Cubic, tangentMode=RCTM_User, time=1, value=1}
          ]}}})
     ))
    ScaleA=true
    ScaleRGB=true
   )
   /SetVariables(Particles.Glow Mesh Color:LinearColor=&MultiplyLinearColor(
     Linear Color=$Particles.Color
     Scale Factor=&MakeFloat4FromVectorAndFloat("Vector (RGB)"=<1 1 1>, "Float (Alpha)"=10)
    )) {isSetParameters=true}
   /SetVariables(
    Particles.ShockwaveColor:LinearColor=&MakeLinearColorFromVectorAndFloat(
     "Vector (RGB)"=&MakeVectorFromLinearColorRGB(Color=<rgba 0.175 0.225 0.25 1>)
     "Float (Alpha)"=&FloatFromCurve(FloatCurve=DataInterface() {propertyValues {Curve {keys [
          {interpMode=RCIM_Cubic, leaveTangentWeight=0.4714, tangentMode=RCTM_User, time=0, value=1}
          {arriveTangentWeight=0.4714, interpMode=RCIM_Cubic, tangentMode=RCTM_User, time=1, value=0}
         ]}}})
    )
    Particles.ShockwaveDynParam:Vector4f=&MakeVector4(
     X=0.5
     Y=1
     Z=&RandomRangeFloat(Minimum=-360, Maximum=360)
     W=&FloatFromCurve(FloatCurve=DataInterface() {propertyValues {Curve {keys [
          {arriveTangent=7.0476, interpMode=RCIM_Cubic, leaveTangent=7.0476, leaveTangentWeight=1.0541, tangentMode=RCTM_User, time=0, value=0}
          {arriveTangentWeight=1.0541, interpMode=RCIM_Cubic, tangentMode=RCTM_User, time=1, value=3}
         ]}}})
    )
   ) {isSetParameters=true}
  ]
  rendererClasses [&NiagaraMeshRendererProperties]
  renderers [
   {rendererClass=&NiagaraMeshRendererProperties, rendererIndex=0}
   {rendererClass=&NiagaraMeshRendererProperties, rendererIndex=1}
   {rendererClass=&NiagaraMeshRendererProperties, rendererIndex=2}
  ]
 }
}
summary {
 systemName=NS_QuickBurst_Speakers
 emitters [
  {emitterName=Pulse_ElectricArcs, rendererClasses [&NiagaraMeshRendererProperties]}
  {emitterName=Fire, rendererClasses [&NiagaraSpriteRendererProperties]}
  {bEnabled=false, emitterName=Fire_LWE, rendererClasses []}
 ]
 userVariables []
}
